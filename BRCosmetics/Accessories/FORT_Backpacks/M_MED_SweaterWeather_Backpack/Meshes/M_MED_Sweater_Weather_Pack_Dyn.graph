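# Physics asset for the Sweater Weather backpack cosmetic, defining physics bodies for its dangling grenades, grenade rings and leaves.
# A large box on the pack bone anchors capsule bodies for six grenades and nine leaves through angle-limited constraints.
# Three ring boxes chain off grenades 02, 05 and 06 with narrow limits, forming two-link chains.
# Grenades swing more widely than leaves, and twist stays tight throughout.
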PhysicsAsset /BRCosmetics/Accessories/FORT_Backpacks/M_MED_SweaterWeather_Backpack/Meshes/M_MED_Sweater_Weather_Pack_Dyn
bodies {
 dyn_grenades_01 [{shapeType=Capsule, length=0.8, radius=3.3, center=<5 0 0>, rotation=<rot 90 0 0>}]
 dyn_grenades_02 [{shapeType=Capsule, length=0.8, radius=3.3, center=<5 0 0>, rotation=<rot 90 0 0>}]
 dyn_grenades_03 [{shapeType=Capsule, length=0.8, radius=3.3, center=<5 0 0>, rotation=<rot 90 0 0>}]
 dyn_grenades_04 [{shapeType=Capsule, length=0.8, radius=3.3, center=<5 0 0>, rotation=<rot 90 0 0>}]
 dyn_grenades_05 [{shapeType=Capsule, length=0.8, radius=3.3, center=<5 0 0>, rotation=<rot 90 0 0>}]
 dyn_grenades_06 [{shapeType=Capsule, length=0.8, radius=3.3, center=<5 0 0>, rotation=<rot 90 0 0>}]
 dyn_grenades_ring_01 [{shapeType=Box, extentX=4, extentY=3, extentZ=1.2, center=<3 0 0>}]
 dyn_grenades_ring_02 [{shapeType=Box, extentX=4, extentY=3, extentZ=1.2, center=<3 0 0>}]
 dyn_grenades_ring_03 [{shapeType=Box, extentX=4, extentY=3, extentZ=1.2, center=<3 0 0>}]
 dyn_leaves_01 [{shapeType=Capsule, length=2, radius=1, center=<0.5 0 0>, rotation=<rot 90 0 0>}]
 dyn_leaves_02 [{shapeType=Capsule, length=2, radius=1, center=<0.5 0 0>, rotation=<rot 90 0 0>}]
 dyn_leaves_03 [{shapeType=Capsule, length=2, radius=1, center=<0.5 0 0>, rotation=<rot 90 0 0>}]
 dyn_leaves_04 [{shapeType=Capsule, length=2, radius=1, center=<0.5 0 0>, rotation=<rot 90 0 0>}]
 dyn_leaves_05 [{shapeType=Capsule, length=2, radius=1, center=<0.5 0 0>, rotation=<rot 90 0 0>}]
 dyn_leaves_06 [{shapeType=Capsule, length=2, radius=1, center=<0.5 0 0>, rotation=<rot 90 0 0>}]
 dyn_leaves_07 [{shapeType=Capsule, length=2, radius=1, center=<0.5 0 0>, rotation=<rot 90 0 0>}]
 dyn_leaves_08 [{shapeType=Capsule, length=2, radius=1, center=<0.5 0 0>, rotation=<rot 90 0 0>}]
 dyn_leaves_09 [{shapeType=Capsule, length=2, radius=1, center=<0.5 0 0>, rotation=<rot 90 0 0>}]
 pack [{shapeType=Box, extentX=30, extentY=5, extentZ=30, center=<0 -3.366 0>}]
}
constraints [
 {bone1Name=dyn_grenades_01, bone2Name=pack, swing1LimitDegrees=30, swing2LimitDegrees=30, twistLimitDegrees=5}
 {bone1Name=dyn_grenades_02, bone2Name=pack, swing1LimitDegrees=30, swing2LimitDegrees=30, twistLimitDegrees=5}
 {bone1Name=dyn_grenades_ring_01, bone2Name=dyn_grenades_02, swing1LimitDegrees=5, swing2LimitDegrees=20, twistLimitDegrees=2}
 {bone1Name=dyn_grenades_03, bone2Name=pack, swing1LimitDegrees=30, swing2LimitDegrees=30, twistLimitDegrees=5}
 {bone1Name=dyn_grenades_04, bone2Name=pack, swing1LimitDegrees=30, swing2LimitDegrees=30, twistLimitDegrees=5}
 {bone1Name=dyn_grenades_06, bone2Name=pack, swing1LimitDegrees=30, swing2LimitDegrees=30, twistLimitDegrees=5}
 {bone1Name=dyn_grenades_ring_03, bone2Name=dyn_grenades_06, swing1LimitDegrees=5, swing2LimitDegrees=20, twistLimitDegrees=2}
 {bone1Name=dyn_grenades_05, bone2Name=pack, swing1LimitDegrees=30, swing2LimitDegrees=30, twistLimitDegrees=5}
 {bone1Name=dyn_grenades_ring_02, bone2Name=dyn_grenades_05, swing1LimitDegrees=5, swing2LimitDegrees=20, twistLimitDegrees=2}
 {bone1Name=dyn_leaves_02, bone2Name=pack, swing1LimitDegrees=10, swing2LimitDegrees=10, twistLimitDegrees=5}
 {bone1Name=dyn_leaves_03, bone2Name=pack, swing1LimitDegrees=10, swing2LimitDegrees=10, twistLimitDegrees=5}
 {bone1Name=dyn_leaves_04, bone2Name=pack, swing1LimitDegrees=10, swing2LimitDegrees=10, twistLimitDegrees=5}
 {bone1Name=dyn_leaves_05, bone2Name=pack, swing1LimitDegrees=10, swing2LimitDegrees=10, twistLimitDegrees=5}
 {bone1Name=dyn_leaves_06, bone2Name=pack, swing1LimitDegrees=10, swing2LimitDegrees=10, twistLimitDegrees=5}
 {bone1Name=dyn_leaves_07, bone2Name=pack, swing1LimitDegrees=10, swing2LimitDegrees=10, twistLimitDegrees=5}
 {bone1Name=dyn_leaves_08, bone2Name=pack, swing1LimitDegrees=10, swing2LimitDegrees=10, twistLimitDegrees=5}
 {bone1Name=dyn_leaves_01, bone2Name=pack, swing1LimitDegrees=10, swing2LimitDegrees=10, twistLimitDegrees=5}
 {bone1Name=dyn_leaves_09, bone2Name=pack, swing1LimitDegrees=10, swing2LimitDegrees=10, twistLimitDegrees=5}
]
